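# UI material for the MimosaXP level-up icon, using alpha-composite blending.
# The shared MF_UI_MimosaXP_LevelUp_Base material function supplies its color and opacity.
# The color is multiplied by itself to square it, then scaled by vertex alpha.
# Opacity passes straight through a reroute without that vertex alpha multiply.
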Material /Game/UI/Foundation/MimosaXP/MimosaXP_LevelUp/M_UI_Icon_MimosaXP_LevelUp
graph {
 properties {
  bAutomaticallySetUsageInEditor=false
  bUsedWithStaticMesh=false
  blendMode=BLEND_AlphaComposite
  materialDomain=MD_UI
  usedShadingModels=Default Lit
 }
 nodes [
  n1 = MaterialFunctionCall() {materialFunction=/Game/UI/Foundation/MimosaXP/MaterialFunction/MF_UI_MimosaXP_LevelUp_Base}
  Multiply(A=VertexColor().A, B=Multiply(A=@n1.Color, B=@n1.Color))
  Reroute(None=@n1.Opacity)
 ]
}
